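# Post-process style AnimBP for the Masako Reign cosmetic body, adding secondary motion to skirt, loop and belt bones.
# It copies the parent mesh pose, runs a PreSim Control Rig driven by movement states, then chains per-bone Anim Dynamics, with thigh spherical limits on lower loop bones.
# Events update those states, compute gravity/RBAN settings that the AnimGraph leaves unread, and apply an alternative skin weight profile when DBNO begins.
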AnimBlueprint /BRCosmetics/Characters/Player/Female/Medium/Bodies/F_MED_MasakoReign/Meshes/F_MED_MasakoReign_AnimBP
target_skeleton=&/F_MED_MasakoReign_Skeleton
graphs [
 AnimGraph {nodes [
   n1 = Control Rig!(
    Source=Copy Pose From Mesh() {bUseAttachedParent=true}
    Alpha
    Crouch:bool=Get OnlyCrouch() -> (OnlyCrouch:bool)
    Sprinting:bool=Get bIsTacticalSprinting()
    BaseBallSlide:double=Get isBaseballSlidling float() -> (isBaseballSlidling float:double)
    KneeSlide:double=Get isKneeSlidling float() -> (isKneeSlidling float:double)
    GrindRail:bool=Get bIsOnGrindRail()
    Skydiving:bool=Get Skydive_floating() -> (Skydive_floating:bool)
    Playing Emote:bool=Get bIsPlayingEmote()
    Jump:double=Get jump() -> (jump:double)
    SurfingGlider:bool=Get gliderIsSurfingSimple() -> (gliderIsSurfingSimple:bool)
    isIdle:double=Get isIdle() -> (isIdle:double)
    FE_Idle:bool=Get bIsInFrontEnd()
    Unarmed:bool=Get Unarmed_Bool() -> (Unarmed_Bool:bool)
    isTargeting:bool=Get bIsTargeting()
    Moving2D:bool=Get bIsMoving2D()
   ) -> (Pose) {controlRigAssetReference {blueprintRigClass=&/F_MED_MasakoReign_PreSim_CtrlRig.F_MED_MasakoReign_PreSim_CtrlRig_C}}
   n2 = Anim Dynamics - Bone: dyn_skirt_fr_r_03!(ComponentPose=Local To Component(LocalPose=@n1), Alpha=1, GravityOverride:Vector="0,0,-980", ExternalForce="0,0,0") -> (Pose) {
    angularDampingOverride=0.9999
    angularSpringConstant=800
    bAngularSpring=true
    bOverrideAngularDamping=true
    bUseGravityOverride=true
    gravityScale=0.35
    simulationSpace=RootRelative
    boundBone {boneName=dyn_skirt_fr_r_03}
    componentAppliedLinearAccClamp=<2000 2000 200>
    componentLinearAccScale=<0.5 0.5 0.5>
    componentLinearVelScale=<0.5 0.5 0.5>
    gravityOverride {z=-980}
    physicsBodyDefinitions [{
      boundBone {boneName=dyn_skirt_fr_r_03}
      boxExtents=<22 32 32>
      constraintSetup {angularLimitsMax=<0 20 2>, angularLimitsMin=<0 0 -2>, angularTarget=<1 0 0>}
      localJointOffset=<4 0 0>
     }]
    simSpaceSettings {maxAngularAcceleration=35, maxAngularVelocity=7, simSpaceAngularAlpha=0.8}
   }
   n3 = Anim Dynamics - Bone: dyn_skirt_fr_r_04!(ComponentPose=@n2, Alpha=1, GravityOverride:Vector="0,0,-980", ExternalForce="0,0,0") -> (Pose) {
    angularDampingOverride=0.9999
    angularSpringConstant=800
    bAngularSpring=true
    bOverrideAngularDamping=true
    bUseGravityOverride=true
    gravityScale=0.35
    simulationSpace=RootRelative
    boundBone {boneName=dyn_skirt_fr_r_04}
    componentAppliedLinearAccClamp=<2000 2000 200>
    componentLinearAccScale=<0.5 0.5 0.5>
    componentLinearVelScale=<0.5 0.5 0.5>
    gravityOverride {z=-980}
    physicsBodyDefinitions [{
      boundBone {boneName=dyn_skirt_fr_r_04}
      boxExtents=<22 32 32>
      constraintSetup {angularLimitsMax=<0 20 2>, angularLimitsMin=<0 0 -2>, angularTarget=<1 0 0>}
      localJointOffset=<4 0 0>
     }]
    simSpaceSettings {maxAngularAcceleration=35, maxAngularVelocity=7, simSpaceAngularAlpha=0.8}
   }
   n4 = Anim Dynamics - Bone: dyn_skirt_fr_r_05!(ComponentPose=@n3, Alpha=1, GravityOverride:Vector="0,0,-980", ExternalForce="0,0,0") -> (Pose) {
    angularDampingOverride=0.9999
    angularSpringConstant=800
    bAngularSpring=true
    bOverrideAngularDamping=true
    bUseGravityOverride=true
    gravityScale=0.35
    simulationSpace=RootRelative
    boundBone {boneName=dyn_skirt_fr_r_05}
    componentAppliedLinearAccClamp=<2000 2000 200>
    componentLinearAccScale=<0.5 0.5 0.5>
    componentLinearVelScale=<0.5 0.5 0.5>
    gravityOverride {z=-980}
    physicsBodyDefinitions [{
      boundBone {boneName=dyn_skirt_fr_r_05}
      boxExtents=<22 32 32>
      constraintSetup {angularLimitsMax=<0 20 2>, angularLimitsMin=<0 0 -2>, angularTarget=<1 0 0>}
      localJointOffset=<4 0 0>
     }]
    simSpaceSettings {maxAngularAcceleration=35, maxAngularVelocity=7, simSpaceAngularAlpha=0.8}
   }
   n5 = Anim Dynamics - Bone: dyn_skirt_fr_r_06!(ComponentPose=@n4, Alpha=1, GravityOverride:Vector="0,0,-980", ExternalForce="0,0,0") -> (Pose) {
    angularDampingOverride=0.9999
    angularSpringConstant=800
    bAngularSpring=true
    bOverrideAngularDamping=true
    bUseGravityOverride=true
    gravityScale=0.35
    simulationSpace=RootRelative
    boundBone {boneName=dyn_skirt_fr_r_06}
    componentAppliedLinearAccClamp=<2000 2000 200>
    componentLinearAccScale=<0.5 0.5 0.5>
    componentLinearVelScale=<0.5 0.5 0.5>
    gravityOverride {z=-980}
    physicsBodyDefinitions [{
      boundBone {boneName=dyn_skirt_fr_r_06}
      boxExtents=<22 32 32>
      constraintSetup {angularLimitsMax=<0 20 2>, angularLimitsMin=<0 0 -2>, angularTarget=<1 0 0>}
      localJointOffset=<4 0 0>
     }]
    simSpaceSettings {maxAngularAcceleration=35, maxAngularVelocity=7, simSpaceAngularAlpha=0.8}
   }
   n6 = Anim Dynamics - Bone: dyn_skirt_fr_c_03!(ComponentPose=@n5, Alpha=1, GravityOverride:Vector="0,0,-980", ExternalForce="0,0,0") -> (Pose) {
    angularDampingOverride=0.9999
    angularSpringConstant=800
    bAngularSpring=true
    bOverrideAngularDamping=true
    bUseGravityOverride=true
    gravityScale=0.35
    simulationSpace=RootRelative
    boundBone {boneName=dyn_skirt_fr_c_03}
    componentAppliedLinearAccClamp=<2000 2000 200>
    componentLinearAccScale=<0.5 0.5 0.5>
    componentLinearVelScale=<0.5 0.5 0.5>
    gravityOverride {z=-980}
    physicsBodyDefinitions [{
      boundBone {boneName=dyn_skirt_fr_c_03}
      boxExtents=<22 32 32>
      constraintSetup {angularLimitsMax=<0 20 2>, angularLimitsMin=<0 0 -2>, angularTarget=<1 0 0>}
      localJointOffset=<4 0 0>
     }]
    simSpaceSettings {maxAngularAcceleration=35, maxAngularVelocity=7, simSpaceAngularAlpha=0.8}
   }
   n7 = Anim Dynamics - Bone: dyn_skirt_fr_c_04!(ComponentPose=@n6, Alpha=1, GravityOverride:Vector="0,0,-980", ExternalForce="0,0,0") -> (Pose) {
    angularDampingOverride=0.9999
    angularSpringConstant=800
    bAngularSpring=true
    bOverrideAngularDamping=true
    bUseGravityOverride=true
    gravityScale=0.35
    simulationSpace=RootRelative
    boundBone {boneName=dyn_skirt_fr_c_04}
    componentAppliedLinearAccClamp=<2000 2000 200>
    componentLinearAccScale=<0.5 0.5 0.5>
    componentLinearVelScale=<0.5 0.5 0.5>
    gravityOverride {z=-980}
    physicsBodyDefinitions [{
      boundBone {boneName=dyn_skirt_fr_c_04}
      boxExtents=<22 32 32>
      constraintSetup {angularLimitsMax=<0 20 2>, angularLimitsMin=<0 0 -2>, angularTarget=<1 0 0>}
      localJointOffset=<4 0 0>
     }]
    simSpaceSettings {maxAngularAcceleration=35, maxAngularVelocity=7, simSpaceAngularAlpha=0.8}
   }
   n8 = Anim Dynamics - Bone: dyn_skirt_fr_c_05!(ComponentPose=@n7, Alpha=1, GravityOverride:Vector="0,0,-980", ExternalForce="0,0,0") -> (Pose) {
    angularDampingOverride=0.9999
    angularSpringConstant=800
    bAngularSpring=true
    bOverrideAngularDamping=true
    bUseGravityOverride=true
    gravityScale=0.35
    simulationSpace=RootRelative
    boundBone {boneName=dyn_skirt_fr_c_05}
    componentAppliedLinearAccClamp=<2000 2000 200>
    componentLinearAccScale=<0.5 0.5 0.5>
    componentLinearVelScale=<0.5 0.5 0.5>
    gravityOverride {z=-980}
    physicsBodyDefinitions [{
      boundBone {boneName=dyn_skirt_fr_c_05}
      boxExtents=<22 32 32>
      constraintSetup {angularLimitsMax=<0 20 2>, angularLimitsMin=<0 0 -2>, angularTarget=<1 0 0>}
      localJointOffset=<4 0 0>
     }]
    simSpaceSettings {maxAngularAcceleration=35, maxAngularVelocity=7, simSpaceAngularAlpha=0.8}
   }
   n9 = Anim Dynamics - Bone: dyn_skirt_fr_c_06!(ComponentPose=@n8, Alpha=1, GravityOverride:Vector="0,0,-980", ExternalForce="0,0,0") -> (Pose) {
    angularDampingOverride=0.9999
    angularSpringConstant=800
    bAngularSpring=true
    bOverrideAngularDamping=true
    bUseGravityOverride=true
    gravityScale=0.35
    simulationSpace=RootRelative
    boundBone {boneName=dyn_skirt_fr_c_06}
    componentAppliedLinearAccClamp=<2000 2000 200>
    componentLinearAccScale=<0.5 0.5 0.5>
    componentLinearVelScale=<0.5 0.5 0.5>
    gravityOverride {z=-980}
    physicsBodyDefinitions [{
      boundBone {boneName=dyn_skirt_fr_c_06}
      boxExtents=<22 32 32>
      constraintSetup {angularLimitsMax=<0 20 2>, angularLimitsMin=<0 0 -2>, angularTarget=<1 0 0>}
      localJointOffset=<4 0 0>
     }]
    simSpaceSettings {maxAngularAcceleration=35, maxAngularVelocity=7, simSpaceAngularAlpha=0.8}
   }
   n10 = Anim Dynamics - Bone: dyn_skirt_fr_l_03!(ComponentPose=@n9, Alpha=1, GravityOverride:Vector="0,0,-980", ExternalForce="0,0,0") -> (Pose) {
    angularDampingOverride=0.9999
    angularSpringConstant=800
    bAngularSpring=true
    bOverrideAngularDamping=true
    bUseGravityOverride=true
    gravityScale=0.35
    simulationSpace=RootRelative
    boundBone {boneName=dyn_skirt_fr_l_03}
    componentAppliedLinearAccClamp=<2000 2000 200>
    componentLinearAccScale=<0.5 0.5 0.5>
    componentLinearVelScale=<0.5 0.5 0.5>
    gravityOverride {z=-980}
    physicsBodyDefinitions [{
      boundBone {boneName=dyn_skirt_fr_l_03}
      boxExtents=<22 32 32>
      constraintSetup {angularLimitsMax=<0 20 2>, angularLimitsMin=<0 0 -2>, angularTarget=<1 0 0>}
      localJointOffset=<4 0 0>
     }]
    simSpaceSettings {maxAngularAcceleration=35, maxAngularVelocity=7, simSpaceAngularAlpha=0.8}
   }
   n11 = Anim Dynamics - Bone: dyn_skirt_fr_l_04!(ComponentPose=@n10, Alpha=1, GravityOverride:Vector="0,0,-980", ExternalForce="0,0,0") -> (Pose) {
    angularDampingOverride=0.9999
    angularSpringConstant=800
    bAngularSpring=true
    bOverrideAngularDamping=true
    bUseGravityOverride=true
    gravityScale=0.35
    simulationSpace=RootRelative
    boundBone {boneName=dyn_skirt_fr_l_04}
    componentAppliedLinearAccClamp=<2000 2000 200>
    componentLinearAccScale=<0.5 0.5 0.5>
    componentLinearVelScale=<0.5 0.5 0.5>
    gravityOverride {z=-980}
    physicsBodyDefinitions [{
      boundBone {boneName=dyn_skirt_fr_l_04}
      boxExtents=<22 32 32>
      constraintSetup {angularLimitsMax=<0 20 2>, angularLimitsMin=<0 0 -2>, angularTarget=<1 0 0>}
      localJointOffset=<4 0 0>
     }]
    simSpaceSettings {maxAngularAcceleration=35, maxAngularVelocity=7, simSpaceAngularAlpha=0.8}
   }
   n12 = Anim Dynamics - Bone: dyn_skirt_fr_l_05!(ComponentPose=@n11, Alpha=1, GravityOverride:Vector="0,0,-980", ExternalForce="0,0,0") -> (Pose) {
    angularDampingOverride=0.9999
    angularSpringConstant=800
    bAngularSpring=true
    bOverrideAngularDamping=true
    bUseGravityOverride=true
    gravityScale=0.35
    simulationSpace=RootRelative
    boundBone {boneName=dyn_skirt_fr_l_05}
    componentAppliedLinearAccClamp=<2000 2000 200>
    componentLinearAccScale=<0.5 0.5 0.5>
    componentLinearVelScale=<0.5 0.5 0.5>
    gravityOverride {z=-980}
    physicsBodyDefinitions [{
      boundBone {boneName=dyn_skirt_fr_l_05}
      boxExtents=<22 32 32>
      constraintSetup {angularLimitsMax=<0 20 2>, angularLimitsMin=<0 0 -2>, angularTarget=<1 0 0>}
      localJointOffset=<4 0 0>
     }]
    simSpaceSettings {maxAngularAcceleration=35, maxAngularVelocity=7, simSpaceAngularAlpha=0.8}
   }
   n13 = Anim Dynamics - Bone: dyn_skirt_bk_l_02!(ComponentPose=@n12, Alpha=0, GravityOverride:Vector="0,0,-980", ExternalForce="0,0,0") -> (Pose) {
    angularDampingOverride=0.9999
    angularSpringConstant=800
    bAngularSpring=true
    bOverrideAngularDamping=true
    bUseGravityOverride=true
    gravityScale=0.35
    simulationSpace=RootRelative
    boundBone {boneName=dyn_skirt_bk_l_02}
    componentAppliedLinearAccClamp=<2000 2000 200>
    componentLinearAccScale=<0.5 0.5 0.5>
    componentLinearVelScale=<0.5 0.5 0.5>
    gravityOverride {z=-980}
    physicsBodyDefinitions [{
      boundBone {boneName=dyn_skirt_bk_l_02}
      boxExtents=<22 32 32>
      constraintSetup {angularLimitsMax=<0 20 2>, angularLimitsMin=<0 0 -2>, angularTarget=<1 0 0>}
      localJointOffset=<4 0 0>
     }]
    simSpaceSettings {maxAngularAcceleration=35, maxAngularVelocity=7, simSpaceAngularAlpha=0.8}
   }
   n14 = Anim Dynamics - Bone: dyn_skirt_bk_l_03!(ComponentPose=@n13, Alpha=1, GravityOverride:Vector="0,0,-980", ExternalForce="0,0,0") -> (Pose) {
    angularDampingOverride=0.9999
    angularSpringConstant=800
    bAngularSpring=true
    bOverrideAngularDamping=true
    bUseGravityOverride=true
    gravityScale=0.35
    simulationSpace=RootRelative
    boundBone {boneName=dyn_skirt_bk_l_03}
    componentAppliedLinearAccClamp=<2000 2000 200>
    componentLinearAccScale=<0.5 0.5 0.5>
    componentLinearVelScale=<0.5 0.5 0.5>
    gravityOverride {z=-980}
    physicsBodyDefinitions [{
      boundBone {boneName=dyn_skirt_bk_l_03}
      boxExtents=<22 32 32>
      constraintSetup {angularLimitsMax=<0 20 2>, angularLimitsMin=<0 0 -2>, angularTarget=<1 0 0>}
      localJointOffset=<4 0 0>
     }]
    simSpaceSettings {maxAngularAcceleration=35, maxAngularVelocity=7, simSpaceAngularAlpha=0.8}
   }
   n15 = Anim Dynamics - Bone: dyn_skirt_bk_l_04!(ComponentPose=@n14, Alpha=1, GravityOverride:Vector="0,0,-980", ExternalForce="0,0,0") -> (Pose) {
    angularDampingOverride=0.9999
    angularSpringConstant=800
    bAngularSpring=true
    bOverrideAngularDamping=true
    bUseGravityOverride=true
    gravityScale=0.35
    simulationSpace=RootRelative
    boundBone {boneName=dyn_skirt_bk_l_04}
    componentAppliedLinearAccClamp=<2000 2000 200>
    componentLinearAccScale=<0.5 0.5 0.5>
    componentLinearVelScale=<0.5 0.5 0.5>
    gravityOverride {z=-980}
    physicsBodyDefinitions [{
      boundBone {boneName=dyn_skirt_bk_l_04}
      boxExtents=<22 32 32>
      constraintSetup {angularLimitsMax=<0 20 2>, angularLimitsMin=<0 0 -2>, angularTarget=<1 0 0>}
      localJointOffset=<4 0 0>
     }]
    simSpaceSettings {maxAngularAcceleration=35, maxAngularVelocity=7, simSpaceAngularAlpha=0.8}
   }
   n16 = Anim Dynamics - Bone: dyn_skirt_bk_c_02!(ComponentPose=@n15, Alpha=0, GravityOverride:Vector="0,0,-980", ExternalForce="0,0,0") -> (Pose) {
    angularDampingOverride=0.9999
    angularSpringConstant=800
    bAngularSpring=true
    bOverrideAngularDamping=true
    bUseGravityOverride=true
    gravityScale=0.35
    simulationSpace=RootRelative
    boundBone {boneName=dyn_skirt_bk_c_02}
    componentAppliedLinearAccClamp=<2000 2000 200>
    componentLinearAccScale=<0.5 0.5 0.5>
    componentLinearVelScale=<0.5 0.5 0.5>
    gravityOverride {z=-980}
    physicsBodyDefinitions [{
      boundBone {boneName=dyn_skirt_bk_c_02}
      boxExtents=<22 32 32>
      constraintSetup {angularLimitsMax=<0 20 2>, angularLimitsMin=<0 0 -2>, angularTarget=<1 0 0>}
      localJointOffset=<4 0 0>
     }]
    simSpaceSettings {maxAngularAcceleration=35, maxAngularVelocity=7, simSpaceAngularAlpha=0.8}
   }
   n17 = Anim Dynamics - Bone: dyn_skirt_bk_c_03!(ComponentPose=@n16, Alpha=1, GravityOverride:Vector="0,0,-980", ExternalForce="0,0,0") -> (Pose) {
    angularDampingOverride=0.9999
    angularSpringConstant=800
    bAngularSpring=true
    bOverrideAngularDamping=true
    bUseGravityOverride=true
    gravityScale=0.35
    simulationSpace=RootRelative
    boundBone {boneName=dyn_skirt_bk_c_03}
    componentAppliedLinearAccClamp=<2000 2000 200>
    componentLinearAccScale=<0.5 0.5 0.5>
    componentLinearVelScale=<0.5 0.5 0.5>
    gravityOverride {z=-980}
    physicsBodyDefinitions [{
      boundBone {boneName=dyn_skirt_bk_c_03}
      boxExtents=<22 32 32>
      constraintSetup {angularLimitsMax=<0 20 2>, angularLimitsMin=<0 0 -2>, angularTarget=<1 0 0>}
      localJointOffset=<4 0 0>
     }]
    simSpaceSettings {maxAngularAcceleration=35, maxAngularVelocity=7, simSpaceAngularAlpha=0.8}
   }
   n18 = Anim Dynamics - Bone: dyn_skirt_bk_c_04!(ComponentPose=@n17, Alpha=1, GravityOverride:Vector="0,0,-980", ExternalForce="0,0,0") -> (Pose) {
    angularDampingOverride=0.9999
    angularSpringConstant=800
    bAngularSpring=true
    bOverrideAngularDamping=true
    bUseGravityOverride=true
    gravityScale=0.35
    simulationSpace=RootRelative
    boundBone {boneName=dyn_skirt_bk_c_04}
    componentAppliedLinearAccClamp=<2000 2000 200>
    componentLinearAccScale=<0.5 0.5 0.5>
    componentLinearVelScale=<0.5 0.5 0.5>
    gravityOverride {z=-980}
    physicsBodyDefinitions [{
      boundBone {boneName=dyn_skirt_bk_c_04}
      boxExtents=<22 32 32>
      constraintSetup {angularLimitsMax=<0 20 2>, angularLimitsMin=<0 0 -2>, angularTarget=<1 0 0>}
      localJointOffset=<4 0 0>
     }]
    simSpaceSettings {maxAngularAcceleration=35, maxAngularVelocity=7, simSpaceAngularAlpha=0.8}
   }
   n19 = Anim Dynamics - Bone: dyn_skirt_bk_c_05!(ComponentPose=@n18, Alpha=1, GravityOverride:Vector="0,0,-980", ExternalForce="0,0,0") -> (Pose) {
    angularDampingOverride=0.9999
    angularSpringConstant=800
    bAngularSpring=true
    bOverrideAngularDamping=true
    bUseGravityOverride=true
    gravityScale=0.35
    simulationSpace=RootRelative
    boundBone {boneName=dyn_skirt_bk_c_05}
    componentAppliedLinearAccClamp=<2000 2000 200>
    componentLinearAccScale=<0.5 0.5 0.5>
    componentLinearVelScale=<0.5 0.5 0.5>
    gravityOverride {z=-980}
    physicsBodyDefinitions [{
      boundBone {boneName=dyn_skirt_bk_c_05}
      boxExtents=<22 32 32>
      constraintSetup {angularLimitsMax=<0 20 2>, angularLimitsMin=<0 0 -2>, angularTarget=<1 0 0>}
      localJointOffset=<4 0 0>
     }]
    simSpaceSettings {maxAngularAcceleration=35, maxAngularVelocity=7, simSpaceAngularAlpha=0.8}
   }
   n20 = Anim Dynamics - Bone: dyn_skirt_bk_r_02!(ComponentPose=@n19, Alpha=0, GravityOverride:Vector="0,0,-980", ExternalForce="0,0,0") -> (Pose) {
    angularDampingOverride=0.9999
    angularSpringConstant=800
    bAngularSpring=true
    bOverrideAngularDamping=true
    bUseGravityOverride=true
    gravityScale=0.35
    simulationSpace=RootRelative
    boundBone {boneName=dyn_skirt_bk_r_02}
    componentAppliedLinearAccClamp=<2000 2000 200>
    componentLinearAccScale=<0.5 0.5 0.5>
    componentLinearVelScale=<0.5 0.5 0.5>
    gravityOverride {z=-980}
    physicsBodyDefinitions [{
      boundBone {boneName=dyn_skirt_bk_r_02}
      boxExtents=<22 32 32>
      constraintSetup {angularLimitsMax=<0 20 2>, angularLimitsMin=<0 0 -2>, angularTarget=<1 0 0>}
      localJointOffset=<4 0 0>
     }]
    simSpaceSettings {maxAngularAcceleration=35, maxAngularVelocity=7, simSpaceAngularAlpha=0.8}
   }
   n21 = Anim Dynamics - Bone: dyn_skirt_bk_r_03!(ComponentPose=@n20, Alpha=1, GravityOverride:Vector="0,0,-980", ExternalForce="0,0,0") -> (Pose) {
    angularDampingOverride=0.9999
    angularSpringConstant=800
    bAngularSpring=true
    bOverrideAngularDamping=true
    bUseGravityOverride=true
    gravityScale=0.35
    simulationSpace=RootRelative
    boundBone {boneName=dyn_skirt_bk_r_03}
    componentAppliedLinearAccClamp=<2000 2000 200>
    componentLinearAccScale=<0.5 0.5 0.5>
    componentLinearVelScale=<0.5 0.5 0.5>
    gravityOverride {z=-980}
    physicsBodyDefinitions [{
      boundBone {boneName=dyn_skirt_bk_r_03}
      boxExtents=<22 32 32>
      constraintSetup {angularLimitsMax=<0 20 2>, angularLimitsMin=<0 0 -2>, angularTarget=<1 0 0>}
      localJointOffset=<4 0 0>
     }]
    simSpaceSettings {maxAngularAcceleration=35, maxAngularVelocity=7, simSpaceAngularAlpha=0.8}
   }
   n22 = Anim Dynamics - Bone: dyn_skirt_bk_r_04!(ComponentPose=@n21, Alpha=1, GravityOverride:Vector="0,0,-980", ExternalForce="0,0,0") -> (Pose) {
    angularDampingOverride=0.9999
    angularSpringConstant=800
    bAngularSpring=true
    bOverrideAngularDamping=true
    bUseGravityOverride=true
    gravityScale=0.35
    simulationSpace=RootRelative
    boundBone {boneName=dyn_skirt_bk_r_04}
    componentAppliedLinearAccClamp=<2000 2000 200>
    componentLinearAccScale=<0.5 0.5 0.5>
    componentLinearVelScale=<0.5 0.5 0.5>
    gravityOverride {z=-980}
    physicsBodyDefinitions [{
      boundBone {boneName=dyn_skirt_bk_r_04}
      boxExtents=<22 32 32>
      constraintSetup {angularLimitsMax=<0 20 2>, angularLimitsMin=<0 0 -2>, angularTarget=<1 0 0>}
      localJointOffset=<4 0 0>
     }]
    simSpaceSettings {maxAngularAcceleration=35, maxAngularVelocity=7, simSpaceAngularAlpha=0.8}
   }
   n23 = Anim Dynamics - Bone: dyn_skirt_bk_r_05!(ComponentPose=@n22, Alpha=1, GravityOverride:Vector="0,0,-980", ExternalForce="0,0,0") -> (Pose) {
    angularDampingOverride=0.9999
    angularSpringConstant=800
    bAngularSpring=true
    bOverrideAngularDamping=true
    bUseGravityOverride=true
    gravityScale=0.35
    simulationSpace=RootRelative
    boundBone {boneName=dyn_skirt_bk_r_05}
    componentAppliedLinearAccClamp=<2000 2000 200>
    componentLinearAccScale=<0.5 0.5 0.5>
    componentLinearVelScale=<0.5 0.5 0.5>
    gravityOverride {z=-980}
    physicsBodyDefinitions [{
      boundBone {boneName=dyn_skirt_bk_r_05}
      boxExtents=<22 32 32>
      constraintSetup {angularLimitsMax=<0 20 2>, angularLimitsMin=<0 0 -2>, angularTarget=<1 0 0>}
      localJointOffset=<4 0 0>
     }]
    simSpaceSettings {maxAngularAcceleration=35, maxAngularVelocity=7, simSpaceAngularAlpha=0.8}
   }
   n24 = Anim Dynamics - Bone: dyn_loop_01!(ComponentPose=@n23, Alpha=1, GravityOverride:Vector="0,0,-980", ExternalForce="0,0,0") -> (Pose) {
    angularDampingOverride=1
    angularSpringConstant=16000
    bAngularSpring=true
    bOverrideAngularDamping=true
    bUseGravityOverride=true
    gravityScale=0.35
    simulationSpace=RootRelative
    boundBone {boneName=dyn_loop_01}
    componentAppliedLinearAccClamp=<3000 3000 3000>
    componentLinearAccScale=<0.5 0.5 0.5>
    componentLinearVelScale=<0.5 0.5 0.5>
    gravityOverride {z=-980}
    physicsBodyDefinitions [{
      boundBone {boneName=dyn_loop_01}
      boxExtents=<20 30 30>
      constraintSetup {angularLimitsMax=<0 0 5>, angularLimitsMin=<0 -40 -5>, angularTarget=<1 0 0>}
      localJointOffset=<4 0 0>
     }]
    simSpaceSettings {maxAngularAcceleration=35, maxAngularVelocity=7, simSpaceAngularAlpha=0.3}
   }
   n25 = Anim Dynamics - Bone: dyn_loop_02!(ComponentPose=@n24, Alpha=1, GravityOverride:Vector="0,0,-980", ExternalForce="0,0,0") -> (Pose) {
    angularDampingOverride=1
    angularSpringConstant=16000
    bAngularSpring=true
    bOverrideAngularDamping=true
    bUseGravityOverride=true
    bUseSphericalLimits=true
    gravityScale=0.35
    simulationSpace=RootRelative
    boundBone {boneName=dyn_loop_02}
    componentAppliedLinearAccClamp=<3000 3000 3000>
    componentLinearAccScale=<0.5 0.5 0.5>
    componentLinearVelScale=<0.5 0.5 0.5>
    gravityOverride {z=-980}
    physicsBodyDefinitions [{
      boundBone {boneName=dyn_loop_02}
      boxExtents=<20 30 30>
      constraintSetup {angularLimitsMax=<0 0 5>, angularLimitsMin=<0 -40 -5>, angularTarget=<1 0 0>}
      localJointOffset=<4 0 0>
     }]
    simSpaceSettings {maxAngularAcceleration=35, maxAngularVelocity=7, simSpaceAngularAlpha=0.3}
    sphericalLimits [{limitRadius=6, drivingBone {boneName=thigh_r}, sphereLocalOffset=<0 5 0>}]
   }
   n26 = Anim Dynamics - Bone: dyn_loop_03!(ComponentPose=@n25, Alpha=1, GravityOverride:Vector="0,0,-980", ExternalForce="0,0,0") -> (Pose) {
    angularDampingOverride=1
    angularSpringConstant=16000
    bAngularSpring=true
    bOverrideAngularDamping=true
    bUseGravityOverride=true
    bUsePlanarLimit=false
    bUseSphericalLimits=true
    gravityScale=0.35
    simulationSpace=RootRelative
    boundBone {boneName=dyn_loop_03}
    componentAppliedLinearAccClamp=<3000 3000 3000>
    componentLinearAccScale=<0.5 0.5 0.5>
    componentLinearVelScale=<0.5 0.5 0.5>
    gravityOverride {z=-980}
    physicsBodyDefinitions [{
      boundBone {boneName=dyn_loop_03}
      boxExtents=<20 30 30>
      constraintSetup {angularLimitsMax=<0 0 10>, angularLimitsMin=<0 -20 -10>, angularTarget=<1 0 0>}
      localJointOffset=<6 0 0>
     }]
    simSpaceSettings {maxAngularAcceleration=35, maxAngularVelocity=7, simSpaceAngularAlpha=0.3}
    sphericalLimits [{limitRadius=6, drivingBone {boneName=thigh_r}, sphereLocalOffset=<10 8 0>}]
   }
   n27 = Anim Dynamics - Bone: dyn_belt_part_l!(ComponentPose=@n26, Alpha=1, GravityOverride:Vector="0,0,-980", ExternalForce="0,0,0") -> (Pose) {
    angularDampingOverride=1
    angularSpringConstant=12000
    bAngularSpring=true
    bOverrideAngularDamping=true
    bUseGravityOverride=true
    gravityScale=0.35
    simulationSpace=RootRelative
    boundBone {boneName=dyn_belt_part_l}
    componentAppliedLinearAccClamp=<3000 3000 3000>
    componentLinearAccScale=<1 1 1>
    componentLinearVelScale=<1 1 1>
    gravityOverride {z=-980}
    physicsBodyDefinitions [{
      boundBone {boneName=dyn_belt_part_l}
      boxExtents=<20 30 30>
      constraintSetup {angularLimitsMax=<0 0 20>, angularLimitsMin=<0 -30 -20>, angularTarget=<1 0 0>}
      localJointOffset=<8 0 0>
     }]
    simSpaceSettings {maxAngularAcceleration=35, maxAngularVelocity=7, simSpaceAngularAlpha=1}
   }
   Output Pose(Result=Component To Local(ComponentPose=@n27)) {layerGroup=None}
  ]}
 EventGraph {nodes [
   Event BlueprintUpdateAnimation()
   > n1 = Sequence!(execute) -> (then_0, then_1, then_2:Exec, then_3:Exec, then_4:Exec)
   > n2 = "Gravity Override State Function (Parameter Asset) (Nativized)"(
    CharacterPartAnimInstance=Self-Reference()
    GameTimeInSeconds=GetGameTimeInSeconds()
    Parameters=&/F_MED_MasakoReign_Parameters
   )
   > Set GravityOverride(GravityOverride=@n2)
   > Set RBAN Settings Out(RBAN Settings Out=@n2.OutRigidBodyAnimNodeParameters)
   n3 = Get bIsCrouching()
   n4 = Get bIsMoving2D()
   Set CrouchWalk(execute=@n1.then_1, CrouchWalk:double="To Float (Boolean)"(InBool=AND Boolean(A=@n3, B=@n4))) -> (Output_Get:double)
   > Set Sprinting(Sprinting:bool=OR Boolean(A=Get bIsTacticalSprinting(), B=Get bIsSprinting())) -> (Output_Get:bool)
   > Set Skydive_floating(Skydive_floating:bool=Get bIsSkydiveFloating()) -> (Output_Get:bool)
   n5 = NOT Boolean(A=OR Boolean!(A=Get bIsSkydiveFloating(), B=Get bIsSkydiveDiveMode(), C:bool=Get bIsParachuteOpened(), D:bool=Get bIsSkyDiving()) -> (ReturnValue))
   n6 = OR Boolean!(A=Get bIsJumping(), B=Get bIsFalling()) -> (ReturnValue)
   > Set jump(jump:double="To Float (Boolean)"(InBool=AND Boolean(A=@n6, B=@n5))) -> (Output_Get:double)
   > n7 = Branch(Condition=OR Boolean!(
     A=Get bIsSkyDiving()
     B=Get bIsCrouching()
     C:bool=Get bIsPlayingEmote()
     D:bool=Get bIsJumping()
     E:bool=Get bIsFalling()
     F:bool=Get bIsMoving2D()
     G:bool=Get bIsSprinting()
     H:bool=Get bIsTacticalSprinting()
     I:bool=Get bIsInFrontEnd()
    ) -> (ReturnValue))
   > Set isIdle(isIdle:double=0) -> (Output_Get:double)
   Set isIdle(execute=@n7.else, isIdle=1)
   n8 = OR Boolean!(A=Get bIsSliding(), B=Get bIsBaseballSliding()) -> (ReturnValue)
   Set OnlyCrouch(execute=@n1.then_4, OnlyCrouch:bool=AND Boolean(A=Get bIsCrouching(), B=NOT Boolean(A=@n8))) -> (Output_Get:bool)
   Event OnCharacterCustomizationLoaded()
   > Set BodySkelPart(BodySkelPart:Skeletal Mesh Component Object Reference=GetPartSkeletalMeshComponent()) -> (Output_Get:Skeletal Mesh Component Object Reference)
   Event OnParachuteOpened()
   > n9 = Switch on EGliderType(Selection=Get PlayerGliderType() -> (PlayerGliderType=HangGlider))
   Set gliderIsSurfingSimple(execute=@n9.SurfingSimple, gliderIsSurfingSimple:bool=true) -> (Output_Get:bool)
   Event OnBeginDBNO() {is=title K2Node_Event}
   > Branch(Condition=Get bIsDBNO())
   > SetSkinWeightProfile!(
    execute
    self:Skinned Mesh Component Object Reference=Get BodySkelPart() -> (BodySkelPart:Skeletal Mesh Component Object Reference)
    InProfileName:Name=Alternative
    InLayer:ESkinWeightProfileLayer Enum=Secondary
   ) -> (then, ReturnValue:bool)
   Event OnEndBaseballSliding() {is=title K2Node_Event}
   > Set isBaseballSlidling float(isBaseballSlidling float:double=0) -> (Output_Get:double)
   Event OnBeginKneeSliding() {is=title K2Node_Event}
   > Set isKneeSlidling float(isKneeSlidling float:double=1) -> (Output_Get:double)
   Event OnParachuteClosed()
   > Set gliderIsSurfingSimple(gliderIsSurfingSimple=false)
   Event OnEndKneeSliding() {is=title K2Node_Event}
   > Set isKneeSlidling float(isKneeSlidling float=0)
   Event OnWeaponChanged()
   > n10 = Cast To FortPlayerAnimInstance(Object=GetAnimInstance(self=GetOwnerSkeletalMeshComponent()))
   > Set Unarmed(Unarmed:double=1) -> (Output_Get:double)
   > Set Unarmed_Bool(Unarmed_Bool:bool=true) -> (Output_Get:bool)
   > n11 = Switch on EFortWeaponCoreAnimation!(execute, Selection:EFortWeaponCoreAnimation Enum=Get WeaponCoreAnim(self=@n10)) -> (
    Melee:Exec
    Pistol:Exec
    Shotgun:Exec
    PaperBlueprint:Exec
    Rifle:Exec
    MeleeOneHand:Exec
    MachinePistol:Exec
    RocketLauncher:Exec
    GrenadeLauncher:Exec
    GoingCommando:Exec
    AssaultRifle:Exec
    TacticalShotgun:Exec
    SniperRifle:Exec
    TrapPlacement:Exec
    ShoulderLauncher:Exec
    AbilityDecoTool:Exec
    Crossbow:Exec
    C4:Exec
    RemoteControl:Exec
    DualWield:Exec
    AR_BullPup:Exec
    AR_ForwardGrip:Exec
    MedPackPaddles:Exec
    SMG_P90:Exec
    AR_DrumGun:Exec
    Consumable_Small:Exec
    Consumable_Large:Exec
    Balloon:Exec
    MountedTurret:Exec
    CreativeTool:Exec
    ExplosiveBow:Exec
    AshtonIndigo:Exec
    AshtonChicago:Exec
    MeleeDualWield:Exec
    Unarmed:Exec
   )
   Set Unarmed(execute=@n11.Unarmed, Unarmed=0)
   > Set Unarmed_Bool(Unarmed_Bool=false)
   Event OnBeginBaseballSliding() {is=title K2Node_Event}
   > Set isBaseballSlidling float(isBaseballSlidling float=1)
   Get bIsMoving2D()
  ]}
]
variables [
 BodySkelPart {type=SkeletalMeshComponent}
 CrouchWalk {type=number}
 gliderIsSurfingSimple {type=boolean}
 GravityOverride {type=Vector}
 isBaseballSlidling float {type=number}
 isIdle {type=number}
 isKneeSlidling float {type=number}
 jump {type=number}
 OnlyCrouch {type=boolean}
 RBAN Settings Out {type=FortRigidBodyAnimNodeParameters}
 Skydive_floating {type=boolean}
 Sprinting {type=boolean}
 Unarmed {type=number}
 Unarmed_Bool {type=boolean}
]
