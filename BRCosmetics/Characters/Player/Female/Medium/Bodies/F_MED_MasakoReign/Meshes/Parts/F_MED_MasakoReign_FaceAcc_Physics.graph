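# Physics asset for the F_MED_MasakoReign face accessory mesh, set up for a dynamic ponytail on this female player character.
# A seven-bone dyn_ponytail_hair chain hangs from a base bone locked to the head, with tapering capsules, tight twist limits and wider swing past the root.
# Capsules on the head, neck, spine, pelvis, clavicles and upper arms form the body volume around the hair, joined by locked or free constraints.
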PhysicsAsset /BRCosmetics/Characters/Player/Female/Medium/Bodies/F_MED_MasakoReign/Meshes/Parts/F_MED_MasakoReign_FaceAcc_Physics
bodies {
 base_ponytail_hair [{shapeType=Capsule, length=0, radius=0, shapeName=base_ponytail_hair_capsule, rotation=<rot -90 0 0>}]
 clavicle_l [{shapeType=Capsule, length=12, radius=9, shapeName=clavicle_l_capsule, center=<10 0.8313 0>, rotation=<rot 0 90 90>}]
 clavicle_r [{shapeType=Capsule, length=12, radius=9, shapeName=clavicle_r_capsule, center=<-10 -0.8313 0>, rotation=<rot 0 90 -90>}]
 dyn_ponytail_hair [{shapeType=Capsule, length=2, radius=5, shapeName=dyn_ponytail_hair_capsule, center=<5 1.5 0>, rotation=<rot -90 0 0>}]
 dyn_ponytail_hair_01 [{shapeType=Capsule, length=4, radius=6, shapeName=dyn_ponytail_hair_01_capsule, center=<5 0 0>, rotation=<rot -90 0 0>}]
 dyn_ponytail_hair_02 [{shapeType=Capsule, length=8, radius=5, shapeName=dyn_ponytail_hair_02_capsule, center=<5 0 0>, rotation=<rot -90 0 0>}]
 dyn_ponytail_hair_03 [{shapeType=Capsule, length=8, radius=4.5, shapeName=dyn_ponytail_hair_03_capsule, center=<5 0 0>, rotation=<rot -90 0 0>}]
 dyn_ponytail_hair_04 [{shapeType=Capsule, length=8, radius=4, shapeName=dyn_ponytail_hair_04_capsule, center=<5 0 0>, rotation=<rot -90 0 0>}]
 dyn_ponytail_hair_05 [{shapeType=Capsule, length=8, radius=3, shapeName=dyn_ponytail_hair_05_capsule, center=<5 0 0>, rotation=<rot -90 0 0>}]
 dyn_ponytail_hair_06 [{shapeType=Capsule, length=8, radius=3, shapeName=dyn_ponytail_hair_06_capsule, center=<5 0 0>, rotation=<rot -90 0 0>}]
 head [{shapeType=Capsule, length=4.5146, radius=6.774, shapeName=head_capsule, center=<4.3298 -1.2709 0>, rotation=<rot 0 59.3677 90>}]
 neck_01 [{shapeType=Capsule, length=8, radius=6, shapeName=neck_01_capsule, center=<0.3821 1.2797 0>, rotation=<rot 90 -16.6254 0>}]
 neck_02 [{shapeType=Capsule, length=8, radius=6, shapeName=neck_02_capsule, rotation=<rot 90 -13.5543 0>}]
 pelvis [{shapeType=Capsule, length=29.8897, radius=19.4824, shapeName=pelvis_capsule, center=<-6.0735 3.8401 0>, rotation=<rot 0 -90 0>}]
 spine_02 [{shapeType=Capsule, length=30.5331, radius=14.4669, shapeName=spine_02_capsule, center=<-10.6029 3.5973 0>}]
 spine_03 [{shapeType=Capsule, length=30.2006, radius=14.7994, shapeName=spine_03_capsule, center=<-5.5688 3.5973 0>}]
 spine_04 [{shapeType=Capsule, length=20, radius=10, shapeName=spine_04_capsule, center=<0 1.9518 0>}]
 spine_05 [{shapeType=Capsule, length=20, radius=10, shapeName=spine_05_capsule, center=<0 1.9518 0>}]
 upperarm_l [{shapeType=Capsule, length=34.1063, radius=9.5931, shapeName=upperarm_l_capsule, center=<22.7862 -0.1952 4.0626>, rotation=<rot 0 90 90>}]
 upperarm_r [{shapeType=Capsule, length=34.1063, radius=9.5931, shapeName=upperarm_r_capsule, center=<-22.7862 0.1952 -4.0626>, rotation=<rot 0 90 -90>}]
}
constraints [
 {
  bone1Name=base_ponytail_hair
  bone2Name=head
  swing1LimitDegrees=45
  swing1Motion=Locked
  swing2LimitDegrees=45
  swing2Motion=Locked
  twistLimitDegrees=45
  twistMotion=Locked
 }
 {
  bone1Name=spine_03
  bone2Name=pelvis
  swing1LimitDegrees=45
  swing1Motion=Locked
  swing2LimitDegrees=45
  swing2Motion=Locked
  twistLimitDegrees=45
  twistMotion=Locked
 }
 {
  bone1Name=spine_04
  bone2Name=spine_03
  swing1LimitDegrees=45
  swing1Motion=Free
  swing2LimitDegrees=45
  swing2Motion=Free
  twistLimitDegrees=45
  twistMotion=Free
 }
 {
  bone1Name=spine_05
  bone2Name=spine_04
  swing1LimitDegrees=45
  swing1Motion=Free
  swing2LimitDegrees=45
  swing2Motion=Free
  twistLimitDegrees=45
  twistMotion=Free
 }
 {
  bone1Name=clavicle_l
  bone2Name=spine_05
  swing1LimitDegrees=45
  swing1Motion=Locked
  swing2LimitDegrees=45
  swing2Motion=Locked
  twistLimitDegrees=45
  twistMotion=Locked
 }
 {
  bone1Name=clavicle_r
  bone2Name=spine_05
  swing1LimitDegrees=45
  swing1Motion=Locked
  swing2LimitDegrees=45
  swing2Motion=Locked
  twistLimitDegrees=45
  twistMotion=Locked
 }
 {
  bone1Name=head
  bone2Name=neck_02
  swing1LimitDegrees=45
  swing1Motion=Free
  swing2LimitDegrees=45
  swing2Motion=Free
  twistLimitDegrees=45
  twistMotion=Free
 }
 {
  bone1Name=upperarm_l
  bone2Name=clavicle_l
  swing1LimitDegrees=45
  swing1Motion=Locked
  swing2LimitDegrees=45
  swing2Motion=Locked
  twistLimitDegrees=45
  twistMotion=Locked
 }
 {
  bone1Name=upperarm_r
  bone2Name=clavicle_r
  swing1LimitDegrees=45
  swing1Motion=Locked
  swing2LimitDegrees=45
  swing2Motion=Locked
  twistLimitDegrees=45
  twistMotion=Locked
 }
 {
  bone1Name=spine_02
  bone2Name=pelvis
  swing1LimitDegrees=45
  swing1Motion=Free
  swing2LimitDegrees=45
  swing2Motion=Free
  twistLimitDegrees=45
  twistMotion=Free
 }
 {
  bone1Name=spine_03
  bone2Name=spine_02
  swing1LimitDegrees=45
  swing1Motion=Free
  swing2LimitDegrees=45
  swing2Motion=Free
  twistLimitDegrees=45
  twistMotion=Free
 }
 {bone1Name=dyn_ponytail_hair, bone2Name=base_ponytail_hair, swing1LimitDegrees=10, swing2LimitDegrees=10, twistLimitDegrees=2}
 {bone1Name=dyn_ponytail_hair_01, bone2Name=dyn_ponytail_hair, swing1LimitDegrees=40, swing2LimitDegrees=20, twistLimitDegrees=2}
 {bone1Name=dyn_ponytail_hair_02, bone2Name=dyn_ponytail_hair_01, swing1LimitDegrees=40, swing2LimitDegrees=40, twistLimitDegrees=2}
 {bone1Name=dyn_ponytail_hair_03, bone2Name=dyn_ponytail_hair_02, swing1LimitDegrees=40, swing2LimitDegrees=40, twistLimitDegrees=2}
 {bone1Name=dyn_ponytail_hair_04, bone2Name=dyn_ponytail_hair_03, swing1LimitDegrees=40, swing2LimitDegrees=40, twistLimitDegrees=2}
 {bone1Name=dyn_ponytail_hair_05, bone2Name=dyn_ponytail_hair_04, swing1LimitDegrees=40, swing2LimitDegrees=40, twistLimitDegrees=2}
 {bone1Name=dyn_ponytail_hair_06, bone2Name=dyn_ponytail_hair_05, swing1LimitDegrees=40, swing2LimitDegrees=40, twistLimitDegrees=2}
 {
  bone1Name=neck_01
  bone2Name=spine_05
  swing1LimitDegrees=45
  swing1Motion=Free
  swing2LimitDegrees=45
  swing2Motion=Free
  twistLimitDegrees=45
  twistMotion=Free
 }
 {
  bone1Name=neck_02
  bone2Name=neck_01
  swing1LimitDegrees=45
  swing1Motion=Free
  swing2LimitDegrees=45
  swing2Motion=Free
  twistLimitDegrees=45
  twistMotion=Free
 }
]
